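# Physics asset for the Banana pickaxe, a two-handed melee cosmetic, giving its skeleton collision bodies and joints for a floppy peel.
# A long capsule covers the handle, three capsules form the banana base, and each peel bone gets one capsule.
# Three four-bone peel chains (a, b, c) are constrained root-to-base and link-to-link, bending mostly on one swing axis with tight twist limits.
# The handle body has no constraint to the base or peels.
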PhysicsAsset /BRCosmetics/Weapons/FORT_Melee/Pickaxe_Banana/Meshes/Pickaxe_Banana_Physics
bodies {
 Melee_TwoHanded_Handle [{shapeType=Capsule, length=113.2651, radius=3.5945, center=<0.8208 -0.1071 22.1929>}]
 dyn_banana_base [
  {shapeType=Capsule, length=14.5588, radius=6.1537, center=<-0.0214 17.9738 3.5334>, rotation=<rot -69.9949 89.9972 179.9999>}
  {shapeType=Capsule, length=12.8734, radius=6.7018, center=<-0.0335 -15.7585 2.8603>, rotation=<rot -74.9996 -90.0094 0.0002>}
  {shapeType=Capsule, length=18.5464, radius=6.5032, center=<0 0 0.8042>, rotation=<rot 0 0 90.0004>}
 ]
 dyn_peel_a01 [{shapeType=Capsule, length=1.6486, radius=3.6959, center=<2.3616 0.7305 -0.0001>, rotation=<rot -90 -178.2101 -176.7898>}]
 dyn_peel_a02 [{shapeType=Capsule, length=2.7543, radius=3.6959, center=<2.9725 0 -0.0003>, rotation=<rot -90 -0.0198 0.0198>}]
 dyn_peel_a03 [{shapeType=Capsule, length=3.2414, radius=3.6959, center=<3.4839 0 0.0001>, rotation=<rot -90 0.0396 -0.0396>}]
 dyn_peel_a04 [{shapeType=Capsule, length=3.2414, radius=3.6959, center=<2.2669 0 -0.0001>, rotation=<rot -90 0.0049 -0.0049>}]
 dyn_peel_b01 [{shapeType=Capsule, length=2.1568, radius=4.1336, center=<1.0197 -0.0001 -0.1007>, rotation=<rot -84.9914 -179.9999 180>}]
 dyn_peel_b02 [{shapeType=Capsule, length=3.7942, radius=3.8854, center=<2.2333 0.0001 -0.0863>, rotation=<rot -84.9946 -179.9999 180>}]
 dyn_peel_b03 [{shapeType=Capsule, length=4.689, radius=4.1336, center=<0.9832 0 0>, rotation=<rot -84.9948 -180 180>}]
 dyn_peel_b04 [{shapeType=Capsule, length=4.689, radius=4.1336, center=<0.6475 0 0>, rotation=<rot -90 0.0264 -0.0263>}]
 dyn_peel_c01 [{shapeType=Capsule, length=2.0396, radius=4.1336, center=<0.9511 -0.0001 0.5034>, rotation=<rot -80 0 0>}]
 dyn_peel_c02 [{shapeType=Capsule, length=3.0525, radius=4.1336, center=<2.1416 0.7738 0.7068>, rotation=<rot -85 0 0>}]
 dyn_peel_c03 [{shapeType=Capsule, length=4.689, radius=4.1336, center=<1.0239 0 0.3798>, rotation=<rot -90 174.9489 -174.9488>}]
 dyn_peel_c04 [{shapeType=Capsule, length=4.689, radius=4.1336, center=<1.0886 0.0001 0>, rotation=<rot -90 2.5305 -2.5305>}]
}
constraints [
 {bone1Name=dyn_peel_a01, bone2Name=dyn_banana_base, swing1LimitDegrees=45, swing2LimitDegrees=1, twistLimitDegrees=1}
 {bone1Name=dyn_peel_a02, bone2Name=dyn_peel_a01, swing1LimitDegrees=30, swing2LimitDegrees=10, twistLimitDegrees=5}
 {bone1Name=dyn_peel_a03, bone2Name=dyn_peel_a02, swing1LimitDegrees=30, swing2LimitDegrees=10, twistLimitDegrees=5}
 {bone1Name=dyn_peel_a04, bone2Name=dyn_peel_a03, swing1LimitDegrees=15, swing2LimitDegrees=5, twistLimitDegrees=1}
 {bone1Name=dyn_peel_b01, bone2Name=dyn_banana_base, swing1LimitDegrees=30, swing2LimitDegrees=1, twistLimitDegrees=1}
 {bone1Name=dyn_peel_b02, bone2Name=dyn_peel_b01, swing1LimitDegrees=30, swing2LimitDegrees=10, twistLimitDegrees=5}
 {bone1Name=dyn_peel_b03, bone2Name=dyn_peel_b02, swing1LimitDegrees=30, swing2LimitDegrees=10, twistLimitDegrees=5}
 {bone1Name=dyn_peel_b04, bone2Name=dyn_peel_b03, swing1LimitDegrees=30, swing2LimitDegrees=10, twistLimitDegrees=5}
 {bone1Name=dyn_peel_c01, bone2Name=dyn_banana_base, swing1LimitDegrees=30, swing2LimitDegrees=1, twistLimitDegrees=1}
 {bone1Name=dyn_peel_c02, bone2Name=dyn_peel_c01, swing1LimitDegrees=30, swing2LimitDegrees=10, twistLimitDegrees=5}
 {bone1Name=dyn_peel_c03, bone2Name=dyn_peel_c02, swing1LimitDegrees=30, swing2LimitDegrees=10, twistLimitDegrees=5}
 {bone1Name=dyn_peel_c04, bone2Name=dyn_peel_c03, swing1LimitDegrees=30, swing2LimitDegrees=10, twistLimitDegrees=5}
]
